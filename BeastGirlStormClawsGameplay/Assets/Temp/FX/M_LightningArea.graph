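# Translucent unlit Niagara material for a lightning area effect, projected as radial UVs onto the scene behind it.
# World position behind translucency, offset by a dynamic-parameter center (X feeds Y too) and divided by Size, becomes polar coordinates.
# Perlin-masked noise warps those UVs.
# Noise-distorted plasma coils with an edge mask form opacity, while a tinted, desaturated texture plus a sharp gradient glow form emissive, both scaled by particle color.
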Material /BeastGirlStormClawsGameplay/Assets/Temp/FX/M_LightningArea
graph {
 properties {
  bUsedWithNiagaraMeshParticles=true
  bUsedWithNiagaraSprites=true
  blendMode=BLEND_Translucent
  pixelDepthOffsetMode=PDOM_Legacy
  shadingModel=MSM_Unlit
  twoSided=true
  usedShadingModels=Unlit
 }
 nodes [
  n1 = TextureObjectParameter() {group=TextureC, parameterName=TextureC_Texture, sortPriority=0, texture=&T_Tile_RainbowColor}
  n2 = VectorParameter() {group=TextureC, parameterName=TextureC_UVControl, channelNames {a=VPan, b=UPan, g=VScale, r=UScale}, defaultValue=<rgba 1 1 0 -0.2>}
  n3 = NamedRerouteUsage()
  n4 = MaterialFunctionCall(Texture=@n1, UVControl_RGB=@n2, UVControl_A=@n2, "[Override]UVInput"=@n3) {materialFunction=&FXCommon_TextureSample}
  n5 = VectorParameter() {parameterName=test, defaultValue=<rgba 0.3 1.425 3 1>}
  Add(A=@n4.RGB, B=@n5.RGB)
  n6 = DynamicParameter() {paramNames [CenterPosX,CenterPosY,CenterPosZ,Size]}
  n7 = Subtract(
   A=MaterialFunctionCall() {materialFunction=&WorldPositionBehindTranslucency}.Result
   B=AppendVector(A=AppendVector(A=@n6.CenterPosX, B=@n6.CenterPosX), B=@n6.CenterPosZ)
  )
  n8 = MaterialFunctionCall(Vector or UVs=Divide(A=ComponentMask(None=@n7) {g=true, r=true}, B=@n6.Size)) {materialFunction=&VectorToRadialValue}
  MaterialFunctionCall!(TexCoord=@n8.Radial Coordinates, ClampChannel=Constant3Vector() {constant=<rgba 1 1 0 0>}) {materialFunction=/Game/Effects/Fort_Effects/MaterialFunctions/MF_FX_TilingMethod}
  n9 = TextureObjectParameter() {
   group=NoiseA
   parameterName=NoiseA_Texture
   samplerType=SAMPLERTYPE_LinearGrayscale
   sortPriority=0
   texture=/Game/Effects/Fort_Effects/Emotes/UltraLight/Texture/T_Cloud_Greyscale
  }
  n10 = VectorParameter() {group=NoiseA, parameterName=NoiseA_UVControl, channelNames {a=VPan, b=UPan, g=VScale, r=UScale}, defaultValue=<rgba 0.1 1 0 -0.4>}
  n11 = VectorParameter() {
   group=NoiseA
   parameterName=NoiseA_NoiseControl
   sortPriority=60
   channelNames {a=VStrength, b=UStrength, g=VOffset, r=UOffset}
   defaultValue=<rgba -0.5 -0.5 0.2 0.01>
  }
  n12 = MaterialFunctionCall!(
   NoiseValue=MaterialFunctionCall(Texture=@n9, UVControl_RGB=@n10, UVControl_A=@n10, "[Override]UVInput"=@n3) {materialFunction=&FXCommon_TextureSample}.RG
   Control_RGB=@n11
   Control_A=@n11
  ) {materialFunction=/Game/Effects/Niagara/MasterMaterials/Functions/Common/FXCommon_NoiseControl}
  n13 = Multiply(A=@n12.NoiseValue, B=ScalarParameter() {defaultValue=0.2, parameterName=fwfwfwf})
  n14 = OneMinus(None=MaterialFunctionCall(Float2=Add(A=@n3, B=@n13)) {materialFunction=&BreakOutFloat2Components}.G)
  n15 = Multiply(
   A=Power(Base=Saturate(None=@n14), Exp=ScalarParameter() {defaultValue=30, parameterName=wfwfw})
   B=Multiply(A=@n5.RGB, B=ScalarParameter() {defaultValue=10, parameterName=aaaa})
  )
  n16 = ScalarParameter() {defaultValue=1, parameterName=W}
  n17 = Multiply(A=Power(Base=Desaturation(None=@n4.RGB, Fraction=@n16)) {constExponent=0.5}, B=@n5.RGB)
  n18 = ParticleColor()
  MaterialFunctionCall(Emissive=Multiply(A=Add(A=@n15, B=@n17), B=@n18.RGB)) {materialFunction=&MF_FX_EmissiveAdjustments}
  n19 = TextureObjectParameter() {
   group=TextureA
   parameterName=TextureA_Texture
   samplerType=SAMPLERTYPE_LinearGrayscale
   sortPriority=0
   texture=/Game/Effects/Fort_Effects/Emotes/UltraLight/Texture/T_PlasmaCoils
  }
  n20 = VectorParameter() {group=TextureA, parameterName=TextureA_UVControl, channelNames {a=VPan, b=UPan, g=VScale, r=UScale}, defaultValue=<rgba 1 1.25 0 -1>}
  n21 = MaterialFunctionCall(Texture=@n19, AdditionalUVOffset=@n12.NoiseValue, UVControl_RGB=@n20, UVControl_A=@n20, "[Override]UVInput"=@n3) {materialFunction=&FXCommon_TextureSample}
  n22 = VectorParameter() {parameterName=Ref, defaultValue=<rgba 3 4 0.4 0.5>}
  n23 = MaterialFunctionCall(Float2=@n3) {materialFunction=&BreakOutFloat2Components}
  n24 = VectorParameter() {group=Y, parameterName=Y_EdgeMask_InvertControl, channelNames {a="", b=Power, g=Multiply, r=Offset}, defaultValue=<rgba 0 3 2 1>}
  n25 = VectorParameter() {group=Y, parameterName=Y_EdgeMask_Control, channelNames {a="", b=Power, g=Multiply, r=Offset}, defaultValue=<rgba 0 15 5 1>}
  Multiply(
   A=@n18.A
   B=Multiply(
    A=MaterialFunctionCall(Texture=@n21.R, RGB=@n22.RGB, A=@n22.RGB) {materialFunction=&MF_FX_Refine}.Result
    B=MaterialFunctionCall!(UorV=@n23.G, Invert_EdgeControl=@n24, NoInvert_EdgeControl=@n25) {materialFunction=/Game/Effects/Niagara/MasterMaterials/Functions/Common/FXCommon_EdgeMask_Ex}.Mask
   )
  )
  n26 = MaterialFunctionCall(Float2=@n8.Radial Coordinates) {materialFunction=&BreakOutFloat2Components}
  n27 = Multiply(A=AppendVector(A=@n26.R, B=OneMinus(None=@n26.G)))
  n28 = TextureObjectParameter() {
   group=NoiseA
   parameterName=NoiseB_Texture
   samplerType=SAMPLERTYPE_LinearColor
   sortPriority=0
   texture=/Engine/Functions/Engine_MaterialFunctions02/ExampleContent/Textures/OffsetNoiseDistanceFields_ForNormals
  }
  n29 = VectorParameter() {group=NoiseA, parameterName=NoiseB_UVControl, channelNames {a=VPan, b=UPan, g=VScale, r=UScale}, defaultValue=<rgba 6 0.1 0.05 -10>}
  n30 = Power(
   Base=MaterialFunctionCall(Texture=@n28, UVControl_RGB=@n29, UVControl_A=@n29, "[Override]UVInput"=@n27) {materialFunction=&FXCommon_TextureSample}.RG
   Exp=ScalarParameter() {defaultValue=0.8, parameterName=sf}
  )
  n31 = VectorParameter() {
   group=NoiseB
   parameterName=NoiseB_NoiseControl
   sortPriority=60
   channelNames {a=VStrength, b=UStrength, g=VOffset, r=UOffset}
   defaultValue=<rgba -0.5 -0.5 0 -0.33>
  }
  n32 = TextureObjectParameter() {
   group=NoiseBMask
   parameterName=NoiseBMask_Texture
   samplerType=SAMPLERTYPE_LinearGrayscale
   sortPriority=0
   texture=&T_SimplePerlin_Greyscale
  }
  n33 = VectorParameter() {group=NoiseBMask, parameterName=NoiseBMask_UVControl, channelNames {a=VPan, b=UPan, g=VScale, r=UScale}, defaultValue=<rgba 1 0.05 0.2 5>}
  n34 = Power(
   Base=MaterialFunctionCall(Texture=@n32, UVControl_RGB=@n33, UVControl_A=@n33, "[Override]UVInput"=@n27) {materialFunction=&FXCommon_TextureSample}.R
   Exp=ScalarParameter() {defaultValue=1.25, parameterName=Power}
  )
  n35 = ScalarParameter() {defaultValue=3, parameterName=CW}
  n36 = Add(
   A=@n27
   B=Multiply(
    A=MaterialFunctionCall(NoiseValue=@n30, Control_RGB=@n31, Control_A=@n31) {materialFunction=&FXCommon_NoiseControl}.NoiseValue
    B=MaterialFunctionCall(In=@n34, Contrast=@n35) {materialFunction=&CheapContrast}.Result
   )
  )
  Multiply(B=MaterialFunctionCall(
    TexCoord=NamedRerouteDeclaration(None=@n36) {name=UV, nodeColor=<rgba 1 0 0.0703 1>}
    ClampChannel=Constant3Vector() {constant=<rgba 1 1 0 0>}
   ) {materialFunction=&MF_FX_TilingMethod}.Result)
  TextureCoordinate()
 ]
}
